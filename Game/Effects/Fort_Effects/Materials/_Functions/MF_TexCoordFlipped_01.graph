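# VFX library material function that outputs a tiled texture coordinate with remapped components, named as a flipped UV utility.
# Texture coordinates are multiplied by U Repeat and V Repeat inputs, then split into components.
# The scaled V component feeds both X and Y of the rebuilt float2, so the scaled U value is discarded rather than swapped, despite the Flipped name.
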MaterialFunction /Game/Effects/Fort_Effects/Materials/_Functions/MF_TexCoordFlipped_01
graph {
 properties {bExposeToLibrary=true, libraryCategoriesText [VFX]}
 nodes [
  n1 = FunctionInput() {bUsePreviewValueAsDefault=true, inputName=U Repeat, inputType=FunctionInput_Scalar, previewValue=<1 0 0 1>}
  n2 = FunctionInput() {bUsePreviewValueAsDefault=true, inputName=V Repeat, inputType=FunctionInput_Scalar, sortPriority=1, previewValue=<1 0 0 1>}
  n3 = MaterialFunctionCall(Float2=Multiply(A=TextureCoordinate(), B=AppendVector(A=@n1, B=@n2))) {materialFunction=&BreakOutFloat2Components}
  FunctionOutput(None=MaterialFunctionCall(X=@n3.G, Y=@n3.G) {materialFunction=&MakeFloat2}.Result)
 ]
}
