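# Animated background material for the Pilgrim item shop screen, layering reddish gradients, smoke, flames and atmospherics.
# Screen-aligned and aspect-corrected UVs drive pulsing radial gradients, panning cosine color ramps, bokeh and a starfield with warped clouds.
# Panned noise textures shaped by gradient masks form the flame, smoke and stage-light effects.
# Two layers of radially mapped, tiled shapes use per-cell random rotation, offset and scale for the "Triagles" mask.
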Material /ItemShop/UI/Art/Pilgrim/M_UI_PilgrimShop_BG
graph {
 properties {pixelDepthOffsetMode=PDOM_Legacy, usedShadingModels=Default Lit}
 nodes [
  AppendVector(A=Constant() {r=0.5}, B=Constant())
  Constant3Vector() {constant=<rgba 0 0.0423 0.7011 1>}
  Constant()
  Constant()
  Constant() {r=3}
  Constant() {r=1.5}
  n1 = MaterialFunctionCall() {materialFunction=&ScreenAlignedUVs}
  n2 = MaterialFunctionCall(Float2=MaterialFunctionCall() {materialFunction=&ScreenResolution}.Visible Resolution) {materialFunction=&BreakOutFloat2Components}
  n3 = AppendVector(A=Divide(A=@n2.R, B=@n2.R), B=Constant() {r=1})
  n4 = ScalarParameter() {defaultValue=-0.596, parameterName=X-Offset}
  n5 = ScalarParameter() {parameterName=Y-Offset}
  n6 = Rotator(Coordinate=Add(A=Multiply(A=@n1."X 100%, Y 100%", B=@n3), B=AppendVector(A=@n4, B=@n5))) {speed=0.069}
  n7 = Subtract(A=@n6) {constB=0.5}
  n8 = Multiply(A=@n7) {constB=-1}
  n9 = Ceil(None=ComponentMask(None=Add(A=@n6) {constB=-0.5}) {g=true})
  n10 = NamedRerouteDeclaration(None=LinearInterpolate(
    A=MaterialFunctionCall(Vector or UVs=@n7) {materialFunction=&VectorToRadialValue}.Radial Coordinates
    B=MaterialFunctionCall(Vector or UVs=@n8) {materialFunction=&VectorToRadialValue}.Radial Coordinates
    Alpha=Saturate(None=@n9)
   )) {name=UVMask, nodeColor=<rgba 0 0 0 1>}
  n11 = ComponentMask(None=@n10) {r=true}
  n12 = ComponentMask(None=@n10) {g=true}
  n13 = Multiply(
   A=MaterialFunctionCall(X=Divide(A=@n11, B=@n12), Y=@n11) {materialFunction=&MakeFloat2}.Result
   B=Constant2Vector() {g=2, r=0.25}
  ) {constB=3}
  n14 = Panner(Coordinate=@n13, Time=Multiply(A=Time(), B=ScalarParameter() {defaultValue=0.55, parameterName=Time-Multiply-Triangles-b})) {speedX=0.069, speedY=0.013}
  n15 = Constant() {r=-0.125}
  n16 = Multiply(A=Add(A=Rotator(Coordinate=@n14, Time=@n15) {speed=6.2832}, B=Constant2Vector() {g=0.125, r=0.125})) {constB=10}
  n17 = Frac(None=@n16)
  n18 = MaterialFunctionCall!(
   PrimaryTextureCoordinates=@n16
   2ndOrderTextureCoordinates
   RandomSeed=Constant() {r=98765}
   NoiseDensityBias
   Bypass Noise Influence Crosstalk
  ) {materialFunction=/Game/Effects/Fort_Effects/Materials/_Functions/MF_RandomUVSample}
  n19 = Multiply(A=LinearInterpolate(Alpha=@n18.NoiseWithRange) {constA=-0.25, constB=0.25}, B=Time())
  n20 = MaterialFunctionCall(
   UVs=Add(A=Rotator(Coordinate=@n17, Time=@n19) {speed=6.2832}, B=LinearInterpolate(Alpha=@n18.NoiseWithRange) {constA=-0.35, constB=0})
   Texture Scale=LinearInterpolate(Alpha=@n18.Noise) {constA=0.125, constB=0.5}
  ) {materialFunction=&ScaleUVsByCenter}
  MaterialFunctionCall!(UVs=@n20.UVs, Corner 01, Corner 02, Corner 03, Glow Min, Glow Max, Outline Thickness) {materialFunction=/FortUILibrary/MaterialFunctions/Shapes/MF_UI_WIP_Triangle}
  n21 = Time()
  n22 = ScalarParameter() {defaultValue=0.35, parameterName=Time-Multiply-Triangles-a}
  n23 = Multiply(A=Panner(Coordinate=@n13, Time=Multiply(A=@n21, B=@n22)) {speedX=0.038, speedY=0.013})
  n24 = Constant() {r=0.125}
  n25 = Multiply(A=Add(A=Rotator(Coordinate=@n23, Time=@n24) {speed=6.2832}, B=Constant2Vector() {g=0.25, r=0.35}) {constB=0.25}) {constB=5}
  n26 = Frac(None=@n25)
  n27 = MaterialFunctionCall(PrimaryTextureCoordinates=@n25, RandomSeed=Constant() {r=1234}) {materialFunction=&MF_RandomUVSample}
  n28 = Multiply(A=LinearInterpolate(Alpha=@n27.NoiseWithRange) {constA=-0.125, constB=0.125}, B=Time())
  n29 = MaterialFunctionCall(
   UVs=Add(A=Rotator(Coordinate=@n26, Time=@n28) {speed=6.2832}, B=LinearInterpolate(Alpha=@n27.NoiseWithRange) {constA=-0.125, constB=0.125})
   Texture Scale=LinearInterpolate(Alpha=@n27.Noise) {constA=0.125, constB=0.5}
  ) {materialFunction=&ScaleUVsByCenter}
  MaterialFunctionCall(UVs=@n29.UVs) {materialFunction=&MF_UI_WIP_Triangle}
  MaterialFunctionCall() {materialFunction=&ScreenAlignedUVs}
  MaterialFunctionCall!(
   Color=NamedRerouteUsage()
   Color Intensity=ScalarParameter() {defaultValue=1.1, group=0 - Global, parameterName=Color Intensity}
   Floor Radius=ScalarParameter() {defaultValue=0.25, group=6 - Floor, parameterName=Floor Radius}
   Floor Density=ScalarParameter() {defaultValue=2, group=6 - Floor, parameterName=Floor Density}
   Floor Metallic=ScalarParameter() {group=6 - Floor, parameterName=Floor Metalic}
   Floor Specular=ScalarParameter() {defaultValue=0.5, group=6 - Floor, parameterName=Floor Spec}
   Floor Roughness=ScalarParameter() {defaultValue=0.5, group=6 - Floor, parameterName=Floor Roughness}
   IsFloor=StaticBoolParameter() {group=6 - Floor, parameterName=IsFloor}
  ) {materialFunction=/Game/UI/Foundation/Environment/Shell/Materials/MaterialFunctions/MF_UI_BP_FloorSetup}
  n30 = Min(A=Constant3Vector() {constant=<rgba 0.7493 0.4168 0.8698 1>}, B=NamedRerouteUsage())
  n31 = LinearInterpolate(
   A=VectorParameter() {parameterName=Colour 1, defaultValue=<rgba 1 0.1719 0.1864 1>}.RGB
   B=VectorParameter() {parameterName=Colour 2, defaultValue=<rgba 0.1732 0 0.2396 1>}.RGB
   Alpha=MaterialFunctionCall(UV Channel=NamedRerouteUsage()) {materialFunction=&LinearGradient}.VGradient
  )
  MaterialFunctionCall(Base=@n30, Blend=@n31) {materialFunction=&Blend_Overlay}
  n32 = NamedRerouteUsage()
  MaterialFunctionCall!(U Repeat=Constant() {r=1}, V Repeat=@n32) {materialFunction=/Game/Effects/Fort_Effects/Materials/_Functions/MF_TexCoordFlipped_01}
  n33 = MaterialFunctionCall(UVs=NamedRerouteUsage(), Texture Scale=Constant2Vector() {g=0.2, r=0.9}) {materialFunction=&ScaleUVsByCenter}
  n34 = Constant() {r=1}
  n35 = Multiply(
   A=Saturate(None=MaterialFunctionCall(UVs=@n33.UVs, Size=@n34) {materialFunction=&MF_UI_Circle}.SDF)
   B=Constant() {r=1}
  )
  n36 = NamedRerouteUsage()
  n37 = Multiply(A=@n36, B=AppendVector(A=Constant() {r=2.2}, B=Constant() {r=1.1}))
  n38 = AppendVector(A=Constant(), B=ScalarParameter() {defaultValue=0.12, parameterName=Flame Plasma Speed})
  n39 = TextureSample(UVs=NamedRerouteDeclaration(None=Panner(Coordinate=@n37, Speed=@n38)) {name=UV Info_2, nodeColor=<rgba 1 0.6328 0 1>}) {samplerType=SAMPLERTYPE_Grayscale, texture=&LowResBlurredNoise}
  n40 = AppendVector(A=Constant() {r=3.3}, B=Constant() {r=0.8})
  n41 = Constant()
  n42 = ScalarParameter() {defaultValue=0.15, parameterName=Flame Pase Speed}
  n43 = NamedRerouteDeclaration(None=Panner(Coordinate=Multiply(A=@n36, B=@n40), Speed=AppendVector(A=@n41, B=@n42))) {name=UV Info_3, nodeColor=<rgba 1 0.6328 0 1>}
  n44 = Multiply(
   A=@n39.RGB
   B=TextureSample(UVs=Add(A=@n43)) {samplerType=SAMPLERTYPE_Grayscale, texture=&LowResBlurredNoise}.RGB
  )
  Max(A=@n35, B=@n44)
  Max()
  n45 = Constant() {r=2}
  n46 = ScalarParameter() {defaultValue=1, parameterName=Box_Angle}
  n47 = NamedRerouteUsage()
  Min(
   A=Min(A=MaterialFunctionCall!(UV=@n32, Corner=@n45, Angle=@n46, Flip, Falloff) {materialFunction=/FortUILibrary/MaterialFunctions/Shapes/MF_UI_BoxAngle}.Result)
   B=@n47
  )
  n48 = NamedRerouteUsage()
  n49 = Constant() {r=-0.35}
  n50 = Constant() {r=0.25}
  n51 = Constant() {r=0.75}
  n52 = Constant() {r=1.35}
  n53 = Add(
   A=MaterialFunctionCall(
    UVs=MaterialFunctionCall!(UV=@n48, Perspective Rotation=@n49, Perspective Amount=@n50) {materialFunction=/FortUILibrary/MaterialFunctions/Basics/MF_UI_Perspective}.Perspective UVs
    Texture Scale=AppendVector(A=@n51, B=@n52)
   ) {materialFunction=&ScaleUVsByCenter}.UVs
   B=AppendVector(A=Constant(), B=Constant() {r=0.2})
  )
  n54 = Multiply(A=Distance(A=ComponentMask(None=@n53) {r=true}, B=Constant() {r=0.5})) {constB=2}
  n55 = OneMinus(None=@n54)
  n56 = Distance(A=ComponentMask(None=@n53) {g=true}, B=Constant() {r=0.5})
  n57 = ScalarParameter() {defaultValue=2, parameterName=X}
  n58 = ScalarParameter() {defaultValue=0.1, parameterName=Y}
  n59 = Max(A=@n57, B=@n58)
  n60 = OneMinus(None=Power(Base=Multiply(A=@n56) {constB=2}, Exp=@n59) {constExponent=3})
  Min(A=@n55, B=@n60)
  n61 = NamedRerouteUsage()
  n62 = TextureSample(UVs=Panner(Coordinate=@n61, Time=Time()) {speedY=0.03}) {texture=/Game/Global/Textures/Default/PointSplatter/T_UI_Starfield}
  n63 = MaterialFunctionCall!(UV=@n61, Warp Center, Warp Intensity=Constant() {r=-0.15}, Warp Radius=Constant() {r=-0.5}) {materialFunction=/FortUILibrary/MaterialFunctions/Effects/MF_UI_Warp}
  n64 = Time()
  n65 = TextureSample(UVs=Multiply(A=Panner(Coordinate=@n63.Warped UV, Time=@n64) {speedY=0.02}) {constB=2.6}) {texture=/Game/Effects/Fort_Effects/Gameplay/Butterfly/T_noise_stars}
  n66 = TextureSample() {texture=/Game/Packages/Fortress_Transylvania/Haunted_Forest/Swamps/Textures/clouds_dust_detail_01}
  n67 = MaterialFunctionCall(
   UV=@n61
   Warp Center=AppendVector(A=Constant() {r=0.5}, B=Constant() {r=0.75})
   Warp Intensity=Constant() {r=0.25}
   Warp Radius=Constant() {r=0.35}
  ) {materialFunction=&MF_UI_Warp}
  n68 = Panner(Coordinate=Multiply(A=@n67.Warped UV) {constB=3}, Time=@n64) {speedX=0.007, speedY=0.02}
  n69 = Panner(Coordinate=Multiply(A=@n67.Warped UV) {constB=2}, Time=@n64) {speedX=-0.008, speedY=-0.02}
  n70 = Add(
   A=Multiply(A=TextureSample(UVs=@n68) {texture=&clouds_dust_detail_01}.RGB) {constB=0.6}
   B=Multiply(A=TextureSample(UVs=@n69) {samplerType=SAMPLERTYPE_Grayscale, texture=/FortUILibrary/UIKit/Buttons/Textures/LowResBlurredNoise}.RGB) {constB=0.3}
  )
  n71 = NamedRerouteUsage()
  n72 = Constant()
  n73 = Constant()
  n74 = MaterialFunctionCall(UV=Add(A=@n71, B=AppendVector(A=@n72, B=@n73)), GradientEnd=Constant() {r=0.5}) {materialFunction=&MF_UI_RadialGradient}
  n75 = Constant() {r=0.25}
  n76 = Constant() {r=0.45}
  n77 = Constant() {r=0.25}
  n78 = Multiply(
   A=Add(A=LinearInterpolate(A=@n62.RGB, B=@n65.RGB, Alpha=@n66.RGB), B=Saturate(None=@n70))
   B=Multiply(
    A=MaterialFunctionCall("Value (0-1)"=@n74.Smooth) {materialFunction=&MF_UI_EaseCurves}.EaseOutSine
    B=MaterialFunctionCall(UVs=@n71, GradientStart=@n75, GradientEnd=@n76, RotationAngle=@n77) {materialFunction=&MF_UI_LinearGradient}.Result
   )
  )
  Multiply(A=Constant3Vector() {constant=<rgba 1 0 0.0284 1>}, B=@n78)
  n79 = Add(A=NamedRerouteUsage(), B=ScalarParameter() {defaultValue=0.05, parameterName=Hotspot Displacement})
  n80 = NamedRerouteUsage()
  n81 = LinearInterpolate(
   A=ScalarParameter() {defaultValue=2.5, group=3 - Gradient Overlay, parameterName=Gradient C Power Min}
   B=ScalarParameter() {defaultValue=3, group=3 - Gradient Overlay, parameterName=Gradient C Power Max}
   Alpha=Divide(A=Multiply(A=@n80)) {constB=0.4}
  )
  n82 = ScalarParameter() {defaultValue=1.5, group=3 - Gradient Overlay, parameterName="Gradient  Intensity"}
  n83 = NamedRerouteUsage()
  n84 = Add(
   A=NamedRerouteUsage()
   B=AppendVector(A=Multiply(A=@n83) {constB=-1}, B=ScalarParameter() {group=3 - Gradient Overlay, parameterName=Gradient C Offset Y})
  )
  n85 = Constant() {r=1}
  NamedRerouteDeclaration(None=Multiply(
    A=NamedRerouteUsage()
    B=MaterialFunctionCall!(UV=@n79, Gradient Power=@n81, Gradient Intensity=@n82, Gradient Offset=@n84, Time=@n85) {materialFunction=/Game/UI/Foundation/Environment/Shell/Materials/MaterialFunctions/MF_UI_BP_PulsingRadialGradient}.Result
   )) {name=Central Gradient Color, nodeColor=<rgba 0 0 0 1>}
  n86 = Constant3Vector() {constant=<rgba 0.1042 0.013 0.0161 1>}
  n87 = NamedRerouteUsage()
  n88 = NamedRerouteUsage()
  n89 = MaterialFunctionCall!(
   Gradient=ComponentMask(None=Panner(Coordinate=@n88) {speedX=0.2}) {r=true}
   a=Constant3Vector() {constant=<rgba 0.1823 0 0.0094 1>}
   b=Constant3Vector() {constant=<rgba 0.0573 0.0013 0.0087 1>}
   c=Constant3Vector() {constant=<rgba 0.0885 0.0071 0.0036 1>}
   d=Constant3Vector() {constant=<rgba 0.1875 0.0704 0.001 1>}
  ) {materialFunction=/FortUILibrary/MaterialFunctions/Color/MF_UI_CosineGradient}
  n90 = ComponentMask(None=MaterialFunctionCall!(RGB=NamedRerouteDeclaration(None=@n89.Result) {name=RotatingGradient, nodeColor=<rgba 0 0 0 1>}) {materialFunction=/Engine/Functions/Engine_MaterialFunctions02/Math/RGBtoHSV}."RGB to HSV (0-1)") {b=true}
  n91 = NamedRerouteUsage()
  n92 = ScalarParameter() {defaultValue=2, group=3 - Gradient Overlay, parameterName=Gradient A Power}
  n93 = ScalarParameter() {defaultValue=2, group=3 - Gradient Overlay, parameterName=Gradient A Intensity}
  n94 = AppendVector(
   A=ScalarParameter() {defaultValue=0.6, group=3 - Gradient Overlay, parameterName=Gradient A Offset X}
   B=ScalarParameter() {defaultValue=0.75, group=3 - Gradient Overlay, parameterName=Gradient A Offset Y}
  )
  n95 = NamedRerouteUsage()
  NamedRerouteDeclaration(None=Multiply(
    A=LinearInterpolate(A=@n86, B=@n87, Alpha=@n90)
    B=MaterialFunctionCall(UV=@n91, Gradient Power=@n92, Gradient Intensity=@n93, Gradient Offset=@n94, Time=@n95) {materialFunction=&MF_UI_BP_PulsingRadialGradient}.Result
   )) {name=Top Gradient Color, nodeColor=<rgba 0 0 0 1>}
  n96 = NamedRerouteUsage()
  n97 = ScalarParameter() {defaultValue=2, group=3 - Gradient Overlay, parameterName=Gradient B Power}
  n98 = ScalarParameter() {defaultValue=0.45, group=3 - Gradient Overlay, parameterName=Gradient B Intensity}
  n99 = AppendVector(
   A=ScalarParameter() {defaultValue=0.5, group=3 - Gradient Overlay, parameterName=Gradient B Offset X}
   B=ScalarParameter() {defaultValue=-0.8, group=3 - Gradient Overlay, parameterName=Gradient B Offset Y}
  )
  n100 = NamedRerouteUsage()
  NamedRerouteDeclaration(None=Multiply(
    A=Constant3Vector() {constant=<rgba 0.5781 0.0345 0.0748 1>}
    B=MaterialFunctionCall(UV=@n96, Gradient Power=@n97, Gradient Intensity=@n98, Gradient Offset=@n99, Time=@n100) {materialFunction=&MF_UI_BP_PulsingRadialGradient}.Result
   )) {name=Bottom Left Gradient Color, nodeColor=<rgba 0 0 0 1>}
  n101 = MaterialFunctionCall(Float2=NamedRerouteUsage()) {materialFunction=&BreakOutFloat2Components}
  n102 = Divide(A=@n101.R, B=Max(A=@n101.G))
  n103 = AppendVector(A=@n102, B=Constant() {r=1})
  NamedRerouteDeclaration(None=Multiply(A=NamedRerouteUsage(), B=@n103)) {name=Squared UV, nodeColor=<rgba 0 0 0 1>}
  n104 = MaterialFunctionCall() {materialFunction=&ScreenAlignedUVs}
  n105 = OneMinus(None=@n104."X 100%, Y 100%")
  n106 = Multiply(A=Add(A=ComponentMask(None=@n105) {g=true}) {constB=0.25}) {constB=1.5}
  n107 = Add(A=ComponentMask(None=@n104."X 100%, Y 100%") {g=true}) {constB=0.25}
  NamedRerouteDeclaration(None=Multiply(A=Saturate(None=@n106), B=Saturate(None=@n107))) {name=TopGradient_1, nodeColor=<rgba 0 0 0 1>}
  n108 = Add(A=@n104."X 100%, Y 100%") {constB=-0.5}
  n109 = OneMinus(None=ComponentMask(None=Abs(None=@n108)) {r=true})
  NamedRerouteDeclaration(None=Saturate(None=Add(A=@n109) {constB=0.125})) {name=MidGradient_1, nodeColor=<rgba 0 0 0 1>}
  n110 = Panner(Coordinate=NamedRerouteUsage()) {speedX=-0.1}
  NamedRerouteDeclaration(None=MaterialFunctionCall(
    Gradient=ComponentMask(None=@n110) {r=true}
    a=Constant3Vector() {constant=<rgba 0.0781 0.0019 0 1>}
    b=Constant3Vector() {constant=<rgba 0.401 0.0125 0.0165 1>}
    c=Constant3Vector() {constant=<rgba 0.474 0.0173 0.054 1>}
    d=Constant3Vector() {constant=<rgba 0.8542 0.0921 0 1>}
   ) {materialFunction=&MF_UI_CosineGradient}.Result) {name=RotatingGradient2, nodeColor=<rgba 0 0 0 1>}
  n111 = Add(A=@n36, B=Multiply(A=@n39.R, B=Constant() {r=0.4}))
  NamedRerouteDeclaration(None=Saturate(None=TextureSample(UVs=@n111) {samplerType=SAMPLERTYPE_Grayscale, texture=&LowResBlurredNoise}.R)) {name=2004 PS 2 Lava_1, nodeColor=<rgba 0 1 0.7188 1>}
  NamedRerouteDeclaration(None=Time()) {name=Time, nodeColor=<rgba 0 0 0 1>}
  n112 = SmoothStep(Min=@n35, Value=@n44)
  NamedRerouteDeclaration(None=SmoothStep(Value=Saturate(None=@n112)) {constMax=0.2}) {name=Small Flame, nodeColor=<rgba 0 0.9062 1 1>}
  n113 = Multiply(
   A=ComponentMask(None=NamedRerouteUsage()) {g=true, r=true}
   B=AppendVector(A=ScalarParameter() {defaultValue=2, parameterName=X}, B=ScalarParameter() {defaultValue=0.1, parameterName=Y})
  )
  n114 = Multiply(A=AppendVector(A=Constant() {r=0.04}, B=Constant() {r=0.015}))
  n115 = NamedRerouteUsage()
  n116 = Constant2Vector() {g=0.5, r=1}
  n117 = Constant()
  n118 = Constant() {r=-0.3}
  n119 = MaterialFunctionCall(
   UVs=Add(
    A=MaterialFunctionCall(UVs=@n115, Texture Scale=@n116) {materialFunction=&ScaleUVsByCenter}.UVs
    B=AppendVector(A=@n117, B=@n118)
   )
   Radius=Constant() {r=0.75}
   Density=Constant() {r=1.8}
  ) {materialFunction=&RadialGradientExponential}
  n120 = MaterialFunctionCall!(
   Base=TextureSample(UVs=Panner(Coordinate=@n113, Speed=@n114) {speedX=0.04, speedY=0.02}) {samplerType=SAMPLERTYPE_LinearColor, texture=/Game/Global/Textures/Default/Tileable_Patterns/T_Tile_Smoke_Small}.R
   Add=@n119.RadialGradientExponential
  ) {materialFunction=/Engine/Functions/Engine_MaterialFunctions02/Math/MultiplyAdd}
  NamedRerouteDeclaration(None=SmoothStep(Value=Multiply(A=@n120.Result, B=@n119.RadialGradientExponential)) {constMax=0.9, constMin=0.2}) {name=Flame Base Mask, nodeColor=<rgba 0 1 0.8125 1>}
  NamedRerouteDeclaration(None=Multiply(A=Subtract(B=@n103)) {constB=0.5}) {name=Squared UV Center Offset, nodeColor=<rgba 0 0 0 1>}
  NamedRerouteDeclaration(None=ScalarParameter() {defaultValue=0.5, group=Glow-Centre, parameterName=Cnt_Glow_x-Offset}) {name=Hotspot X-Offset, nodeColor=<rgba 0 0 0 1>}
  NamedRerouteDeclaration(None=ScalarParameter() {defaultValue=1, group=fx-streaks, parameterName=bIsUseStreak}) {name=FX Vis Control, nodeColor=<rgba 0 0 0 1>}
  n121 = NamedRerouteUsage()
  n122 = MaterialFunctionCall(UV Channel=NamedRerouteUsage()) {materialFunction=&LinearGradient}
  NamedRerouteDeclaration(None=LinearInterpolate(
    A=VectorParameter() {parameterName=Flame Colour BG, defaultValue=<rgba 0.5647 0.016 0 1>}.RGB
    B=VectorParameter() {parameterName=Flame Colour Base, defaultValue=<rgba 0.4896 0 0.0765 1>}.RGB
    Alpha=Min(A=@n121, B=@n122.VGradient)
   )) {name=Colours, nodeColor=<rgba 0.0781 1 0 1>}
  n123 = NamedRerouteUsage()
  n124 = NamedRerouteUsage()
  n125 = Constant() {r=0.5}
  n126 = ScalarParameter() {defaultValue=0.4, parameterName=SmoothStep Value}
  n127 = NamedRerouteUsage()
  n128 = StaticSwitchParameter(
   True=NamedRerouteDeclaration(None=MaterialFunctionCall(DistanceField=@n124, Threshold=@n125) {materialFunction=&MF_UI_AntiAliasing}.Outer) {name=Drips, nodeColor=<rgba 0.5938 1 0 1>}
   False=StaticSwitchParameter(True=SmoothStep(Min=@n126, Value=@n127) {constMax=0.9, constMin=0.4}, False=@n127) {parameterName=SmoothSteps}
  ) {parameterName=Drips}
  n129 = Min(A=NamedRerouteUsage(), B=@n128)
  n130 = Subtract(A=@n129, B=OneMinus(None=@n129))
  n131 = LinearInterpolate(
   A=LinearInterpolate(A=NamedRerouteUsage(), B=NamedRerouteUsage(), Alpha=Multiply(A=@n123) {constB=0.1})
   B=NamedRerouteUsage()
   Alpha=Multiply(A=Saturate(None=@n130)) {constA=0.3, constB=0.15}
  )
  n132 = NamedRerouteUsage()
  n133 = Min(A=@n128, B=NamedRerouteUsage())
  n134 = OneMinus(None=@n133)
  n135 = Multiply(A=Saturate(None=Subtract(A=@n133, B=@n134))) {constB=0.3}
  n136 = NamedRerouteUsage()
  NamedRerouteDeclaration(None=LinearInterpolate(A=LinearInterpolate(A=@n131, B=@n132, Alpha=@n135), B=Constant3Vector() {constant=<rgba 1 0 0.0284 1>}, Alpha=Multiply(A=@n136) {constB=0.2})) {name=PhoneCharge Effect, nodeColor=<rgba 1 0.4453 0 1>}
  n137 = MaterialFunctionCall!(ScreenUV, TextureSize, ScalePivotPoint) {materialFunction=/FortUILibrary/MaterialFunctions/Basics/MF_UI_ScreenSpaceUVScaling}
  NamedRerouteDeclaration(None=@n137.Scaled to Fit) {name=UVFill, nodeColor=<rgba 1 0 0.7031 1>}
  NamedRerouteDeclaration(None=MaterialFunctionCall!(UVs=@n137.Scaled to Fit, Scale=Constant() {r=1}, X Position=Constant() {r=-0.6}, Y Position=Constant()) {materialFunction=/OfferCatalog/Art/_MaterialFunctions_ShopTheming/MF_UI_ScalePosition}.Result) {name=SmokeUV, nodeColor=<rgba 0 1 0.0625 1>}
  n138 = MaterialFunctionCall(UV Channel=@n36) {materialFunction=&LinearGradient}
  n139 = Saturate(None=SmoothStep(Value=Saturate(None=@n138.VGradient)) {constMax=0.9})
  n140 = Multiply(A=@n36, B=AppendVector(A=Constant() {r=1.2}, B=Constant() {r=0.8}))
  n141 = AppendVector(A=Constant(), B=ScalarParameter() {defaultValue=0.12, parameterName=Flame Plasma Speed})
  n142 = TextureSample(UVs=NamedRerouteDeclaration(None=Panner(Coordinate=@n140, Speed=@n141)) {name=UV Info, nodeColor=<rgba 1 0.6328 0 1>}) {samplerType=SAMPLERTYPE_Grayscale, texture=/Game/Global/Textures/Default/Noise/LowResBlurredNoise}
  n143 = Multiply(A=@n36, B=AppendVector(A=Constant() {r=2.3}, B=Constant() {r=0.5}))
  n144 = AppendVector(A=Constant(), B=ScalarParameter() {defaultValue=0.15, parameterName=Flame Pase Speed})
  n145 = Add(
   A=NamedRerouteDeclaration(None=Panner(Coordinate=@n143, Speed=@n144)) {name=UV Info_1, nodeColor=<rgba 1 0.6328 0 1>}
   B=Multiply(A=NamedRerouteUsage(), B=Constant() {r=0.2})
  )
  n146 = SmoothStep(
   Min=OneMinus(None=LinearInterpolate(Alpha=@n139) {constA=0.9, constB=1.25})
   Value=Multiply(
    A=@n142.RGB
    B=TextureSample(UVs=@n145) {samplerType=SAMPLERTYPE_Grayscale, texture=&LowResBlurredNoise}.RGB
   )
  )
  NamedRerouteDeclaration(None=SmoothStep(Value=Saturate(None=@n146)) {constMax=0.2}) {name=Cian's Smoke, nodeColor=<rgba 1 0 0.1641 1>}
  NamedRerouteDeclaration(None=@n78) {name=Atmos, nodeColor=<rgba 1 0 0.0703 1>}
  n147 = Multiply(
   A=ComponentMask(None=NamedRerouteUsage()) {g=true, r=true}
   B=AppendVector(A=ScalarParameter() {defaultValue=2, parameterName=X}, B=ScalarParameter() {defaultValue=0.1, parameterName=Y})
  )
  n148 = Multiply(A=AppendVector(A=Constant() {r=0.04}, B=Constant() {r=0.012}))
  n149 = NamedRerouteUsage()
  n150 = Constant2Vector() {g=0.65, r=1.25}
  n151 = Constant()
  n152 = Constant()
  n153 = MaterialFunctionCall(
   UVs=Add(
    A=MaterialFunctionCall(UVs=@n149, Texture Scale=@n150) {materialFunction=&ScaleUVsByCenter}.UVs
    B=AppendVector(A=@n151, B=@n152)
   )
   Radius=Constant() {r=0.75}
   Density=Constant() {r=1.8}
  ) {materialFunction=&RadialGradientExponential}
  n154 = MaterialFunctionCall(
   Base=TextureSample(UVs=Panner(Coordinate=@n147, Speed=@n148) {speedX=0.04, speedY=0.02}) {samplerType=SAMPLERTYPE_LinearColor, texture=&T_Tile_Smoke_Small}.R
   Add=@n153.RadialGradientExponential
  ) {materialFunction=&MultiplyAdd}
  NamedRerouteDeclaration(None=SmoothStep(Value=Multiply(A=@n154.Result, B=@n153.RadialGradientExponential)) {constMax=0.9, constMin=0.2}) {name=Big Flame Mask, nodeColor=<rgba 0 1 0.8125 1>}
  n155 = Add(A=@n36, B=Multiply(A=@n142.R, B=Constant() {r=0.4}))
  NamedRerouteDeclaration(None=Saturate(None=TextureSample(UVs=@n155) {samplerType=SAMPLERTYPE_Grayscale, texture=/FortUILibrary/UIKit/Buttons/Textures/LowResBlurredNoise}.R)) {name=2004 PS 2 Lava, nodeColor=<rgba 0 1 0.7188 1>}
  n156 = NamedRerouteUsage()
  n157 = MaterialFunctionCall(UV Channel=NamedRerouteUsage()) {materialFunction=&LinearGradient}
  NamedRerouteDeclaration(None=LinearInterpolate(
    A=Constant3Vector() {constant=<rgba 0.0469 0.0043 0.0242 1>}
    B=Constant3Vector() {constant=<rgba 0.2344 0.0352 0.1994 1>}
    Alpha=Min(A=@n156, B=@n157.VGradient)
   )) {name=Colours_1, nodeColor=<rgba 0.0781 1 0 1>}
  n158 = ComponentMask(None=@n53) {g=true, r=true}
  n159 = AppendVector(A=@n57, B=@n58)
  n160 = TextureSample(UVs=Panner(Coordinate=Multiply(A=@n158, B=@n159)) {speedX=0.03, speedY=0.02}) {samplerType=SAMPLERTYPE_LinearColor, texture=&T_Tile_Smoke_Small}
  n161 = Multiply(A=@n55, B=@n60)
  n162 = Multiply(
   A=MaterialFunctionCall(
    Base=MaterialFunctionCall("Value (0-1)"=@n160.R) {materialFunction=&MF_UI_EaseCurves}.EaseInSine
    Add=@n161
   ) {materialFunction=&MultiplyAdd}.Result
   B=@n161
  )
  NamedRerouteDeclaration(None=Saturate(None=@n162)) {name=Light_Effect, nodeColor=<rgba 0.3125 1 0 1>}
  n163 = Constant2Vector()
  n164 = Multiply(A=MaterialFunctionCall(UVs=Add(A=@n32, B=@n163), Size=Constant2Vector() {g=3, r=0.75}) {materialFunction=&MF_UI_Box}.SDF) {constB=-1}
  NamedRerouteDeclaration(None=Min(A=@n47, B=SmoothStep(Value=@n164) {constMax=0.1})) {name=Stage Lights, nodeColor=<rgba 0.125 0 1 1>}
  NamedRerouteDeclaration(None=Add(A=@n31, B=@n30)) {name=Stage Colors, nodeColor=<rgba 0 0.0859 1 1>}
  n165 = NamedRerouteUsage()
  n166 = NamedRerouteUsage()
  n167 = VectorParameter() {group=1 - Chrome Visual, parameterName=Color C, sortPriority=3, defaultValue=<rgba 0.099 0 0.0019 1>}
  n168 = Desaturation(None=NamedRerouteUsage(), Fraction=Constant() {r=0.5})
  n169 = VectorParameter() {group=1 - Chrome Visual, parameterName=Color B, sortPriority=2, defaultValue=<rgba 0.2552 0.0077 0.0179 1>}
  n170 = Power(
   Base=Desaturation(None=NamedRerouteUsage(), Fraction=Constant() {r=1})
   Exp=ScalarParameter() {defaultValue=0.4, parameterName=Rear Gradient Power}
  )
  n171 = NamedRerouteUsage()
  n172 = Constant() {r=1}
  n173 = Power(
   Base=Clamp(None=Desaturation(None=@n171, Fraction=@n172), Max=ScalarParameter() {defaultValue=0.6, parameterName=Max Clamp})
   Exp=ScalarParameter() {defaultValue=1.2, parameterName=Hotspot Gradient Power}
  )
  n174 = Add(
   A=LinearInterpolate(A=@n167.RGB, B=Add(A=@n168, B=@n169.RGB), Alpha=Saturate(None=@n170))
   B=LinearInterpolate(
    B=VectorParameter() {group=3 - Gradient Overlay, parameterName=Gradient C Color, defaultValue=<rgba 1 0.1458 0.1557 1>}.RGB
    Alpha=Saturate(None=@n173)
   )
  )
  n175 = NamedRerouteUsage()
  n176 = MaterialFunctionCall!(Base=Add(A=Add(A=@n174, B=@n175), B=NamedRerouteUsage()), Blend=NamedRerouteUsage()) {materialFunction=/Engine/Functions/Engine_MaterialFunctions03/Blends/Blend_SoftLight}
  n177 = Multiply(
   A=NamedRerouteUsage()
   B=VectorParameter() {group=3 - Gradient Overlay, parameterName=Gradient C Color, defaultValue=<rgba 1 0.1458 0.1557 1>}.RGB
  )
  NamedRerouteDeclaration(None=Multiply(A=Multiply(A=@n165, B=@n166), B=Add(A=@n176.Result, B=@n177))) {name=Final Color, nodeColor=<rgba 0 0 0 1>}
  NamedRerouteDeclaration(None=MaterialFunctionCall!(
    UV=NamedRerouteUsage()
    Bokeh Texture
    Bokeh Color=VectorParameter() {group=4 - Bokeh, parameterName=Bokeh Color, defaultValue=<rgba 1 1 1 1>}.RGB
    Bokeh Opacity=ScalarParameter() {defaultValue=0.0312, group=4 - Bokeh, parameterName=Bokeh Opacity}
    Bokeh Power
    Bokeh Scale
    Foreground Scale
    Background Scale
    Bokeh Angle
    Bokeh Speed
    Falloff Intensity
    Falloff Edge
   ) {materialFunction=/Game/UI/Foundation/Environment/Shell/Materials/MaterialFunctions/MF_UI_BP_Bokeh}.Result) {name=Bokeh, nodeColor=<rgba 0 0 0 1>}
  n178 = Multiply(A=NamedRerouteUsage(), B=NamedRerouteUsage())
  n179 = Power(
   Base=Multiply(A=OneMinus(None=@n178), B=ScalarParameter() {defaultValue=1.75, parameterName=Mask-Intensity}) {constB=4}
   Exp=ScalarParameter() {defaultValue=6, parameterName=Mask-Power}
  )
  n180 = OneMinus(None=MaterialFunctionCall() {materialFunction=&ScreenAlignedUVs}."X 100%, Y 100%")
  n181 = Saturate(None=Subtract(A=ComponentMask(None=@n180) {g=true}, B=ScalarParameter() {defaultValue=0.25, parameterName=Gradient Offset}))
  n182 = ComponentMask(None=NamedRerouteUsage()) {g=true}
  n183 = Multiply(A=@n27.NoiseWithRange, B=Saturate(None=OneMinus(None=@n182)))
  n184 = LinearInterpolate(Alpha=@n182) {constA=0.02, constB=0.16}
  n185 = MaterialFunctionCall(UVs=@n29.UVs) {materialFunction=&MF_UI_Box}
  n186 = OneMinus(None=Clamp(None=@n185.SDFOutline))
  n187 = Add(
   A=LinearInterpolate(Alpha=SmoothStep(Max=@n184, Value=@n185.SDFOutline) {constMax=0.02}) {constA=1, constB=0}
   B=Multiply(A=Subtract(A=@n186) {constB=0.85}) {constB=0.5}
  )
  n188 = ComponentMask(None=NamedRerouteUsage()) {g=true}
  n189 = Multiply(A=@n18.NoiseWithRange, B=Saturate(None=OneMinus(None=@n188)))
  n190 = LinearInterpolate(Alpha=@n188) {constA=0.02, constB=0.16}
  n191 = MaterialFunctionCall(UVs=@n20.UVs) {materialFunction=&MF_UI_Box}
  n192 = OneMinus(None=Clamp(None=@n191.SDFOutline))
  n193 = Add(
   A=LinearInterpolate(Alpha=SmoothStep(Max=@n190, Value=@n191.SDFOutline) {constMax=0.02}) {constA=1, constB=0}
   B=Multiply(A=Subtract(A=@n192) {constB=0.85}) {constB=0.5}
  )
  n194 = Max(A=Multiply(A=Saturate(None=@n183), B=Saturate(None=@n187)), B=Multiply(A=Saturate(None=@n189), B=Saturate(None=@n193)))
  NamedRerouteDeclaration(None=Multiply(A=Saturate(None=@n179), B=Multiply(A=@n181, B=@n194))) {name=Triagles, nodeColor=<rgba 0 0 0 1>}
  NamedRerouteDeclaration(None=AppendVector(A=Multiply(A=@n102) {constB=0.5}, B=Constant() {r=0.5})) {name=Center Point, nodeColor=<rgba 0 0 0 1>}
  n195 = Sine(None=Multiply(A=NamedRerouteUsage(), B=ScalarParameter() {defaultValue=0.125, group=3 - Gradient Overlay, parameterName=Gradient Time Factor}))
  NamedRerouteDeclaration(None=LinearInterpolate(Alpha=ConstantBiasScale(None=@n195)) {constA=0.5}) {name=Pulsing Time, nodeColor=<rgba 0 0 0 1>}
  NamedRerouteUsage()
  NamedRerouteUsage()
  NamedRerouteUsage()
  NamedRerouteUsage()
  OneMinus(None=MaterialFunctionCall(UVs=NamedRerouteUsage(), GradientStart=Constant() {r=0.2}, GradientEnd=Constant() {r=0.6}, RotationAngle=Constant() {r=0.45}) {materialFunction=&MF_UI_LinearGradient}.Result)
  Power(Base=@n54, Exp=@n59) {constExponent=3}
  SmoothStep(Value=@n162) {constMin=0.1}
  SmoothStep(Value=MaterialFunctionCall(UV Channel=@n32) {materialFunction=&LinearGradient}.VGradient) {constMax=0.3}
  TextureCoordinate()
  TextureCoordinate()
  VectorParameter() {parameterName=BaseColor3, defaultValue=<rgba 0.0307 0.0212 0.2664 1>}
  VectorParameter() {parameterName=BaseColor4, defaultValue=<rgba 0.3005 0.0669 0.2039 1>}
  VectorParameter() {parameterName=Atmospherics Color, defaultValue=<rgba 1 1 1 1>}
  VectorParameter() {group=3 - Gradient Overlay, parameterName=Gradient A Color, defaultValue=<rgba 0.0768 0.0242 0.224 1>}
  VectorParameter() {group=3 - Gradient Overlay, parameterName=Gradient B Color, defaultValue=<rgba 0.296 0.083 0.5781 1>}
  VectorParameter() {group=3 - Gradient Overlay, parameterName=Gradient C Color, defaultValue=<rgba 1 0.1458 0.1557 1>}
  VectorParameter() {parameterName=BaseColor4, defaultValue=<rgba 0.3005 0.0669 0.2039 1>}
  VectorParameter() {parameterName=Highlight, defaultValue=<rgba 1 0.5111 0.474 1>}
  VectorParameter() {group=3 - Gradient Overlay, parameterName=Gradient A Color, defaultValue=<rgba 0.0768 0.0242 0.224 1>}
  VectorParameter() {group=3 - Gradient Overlay, parameterName=Gradient B Color, defaultValue=<rgba 0.296 0.083 0.5781 1>}
  VectorParameter() {group=1 - Chrome Visual, parameterName=Color B, sortPriority=2, defaultValue=<rgba 0.2552 0.0077 0.0179 1>}
  VectorParameter() {parameterName=BaseColor, defaultValue=<rgba 0.117 0.0307 0.3185 1>}
  VectorParameter() {parameterName=BaseColor2, defaultValue=<rgba 0.0187 0.0606 0.3229 1>}
 ]
}
